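# Physics asset for the face accessory mesh of the large male TennisLeashHood character body, covering its beard and hood bones.
# Capsule bodies sit on three beard bones, a beard base, three hood bones, and the head, upper spine and left clavicle.
# Constraints lock the angular motion of side beard bones to the main beard, upper hood to head, and side hood bones to spine_05.
# Only the beard-to-base joint leaves swing and twist unlocked.
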PhysicsAsset /BRCosmetics/Characters/Player/Male/Large/Bodies/M_LRG_TennisLeashHood/Meshes/Parts/M_LRG_TennisLeashHood_FaceAcc_Physics
bodies {
 clavicle_l [{
   shapeType=Capsule
   length=9.255
   radius=9.505
   shapeName=clavicle_l_capsule
   center=<11.8282 -0.315 1.715>
   rotation=<rot 1.0024 -89.6888 103.2409>
  }]
 dyn_beard [{shapeType=Capsule, length=16.5, radius=2.75, shapeName=dyn_beard_capsule, center=<9.3773 2.3035 0>}]
 dyn_beard_base [{shapeType=Capsule, length=0.505, radius=2.005, shapeName=dyn_beard_base_capsule, rotation=<rot 0 -90 0>}]
 dyn_beard_l [{shapeType=Capsule, length=9.5, radius=2.5, shapeName=dyn_beard_l_capsule, rotation=<rot 90 -123.0222 -123.0225>}]
 dyn_beard_r [{shapeType=Capsule, length=9.5, radius=2.5, shapeName=dyn_beard_r_capsule, rotation=<rot 90 -123.0222 -123.0225>}]
 dyn_hood_l [{shapeType=Capsule, length=21.25, radius=5.75, shapeName=dyn_hood_l_capsule, center=<3.0238 7.2961 -1.7035>, rotation=<rot -75 -44.9986 -25>}]
 dyn_hood_r [{shapeType=Capsule, length=19.5, radius=4.75, shapeName=dyn_hood_r_capsule, center=<-2.0595 -7.4628 3.1877>, rotation=<rot -85 -44.9991 -20>}]
 dyn_hood_up [{shapeType=Capsule, length=25.5, radius=6.5, shapeName=dyn_hood_up_capsule}]
 head [{shapeType=Capsule, length=8, radius=8.5, shapeName=head_capsule, center=<4.3463 -1.1646 0>, rotation=<rot -90 0 -15.0002>}]
 spine_05 [{shapeType=Capsule, length=16.5, radius=15, shapeName=spine_05_capsule, center=<0 2.5958 0>}]
}
constraints [
 {
  bone1Name=dyn_beard_r
  bone2Name=dyn_beard
  swing1LimitDegrees=45
  swing1Motion=Locked
  swing2LimitDegrees=45
  swing2Motion=Locked
  twistLimitDegrees=45
  twistMotion=Locked
 }
 {
  bone1Name=dyn_beard_l
  bone2Name=dyn_beard
  swing1LimitDegrees=45
  swing1Motion=Locked
  swing2LimitDegrees=45
  swing2Motion=Locked
  twistLimitDegrees=45
  twistMotion=Locked
 }
 {
  bone1Name=dyn_hood_up
  bone2Name=head
  swing1LimitDegrees=10
  swing1Motion=Locked
  swing2LimitDegrees=10
  swing2Motion=Locked
  twistLimitDegrees=10
  twistMotion=Locked
 }
 {
  bone1Name=dyn_hood_l
  bone2Name=spine_05
  swing1LimitDegrees=45
  swing1Motion=Locked
  swing2LimitDegrees=45
  swing2Motion=Locked
  twistLimitDegrees=45
  twistMotion=Locked
 }
 {
  bone1Name=dyn_hood_r
  bone2Name=spine_05
  swing1LimitDegrees=45
  swing1Motion=Locked
  swing2LimitDegrees=45
  swing2Motion=Locked
  twistLimitDegrees=45
  twistMotion=Locked
 }
 {bone1Name=dyn_beard, bone2Name=dyn_beard_base, swing1LimitDegrees=60, swing2LimitDegrees=20, twistLimitDegrees=30}
]
