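# Control Rig for the TennisLeashHood male character's face accessories, procedurally adjusting the glasses and hood bones.
# A parent constraint makes the glasses partially follow both outer cheek bones, then a sequence runs three spherical pose readers on the head.
# Their outputs weight transform offsets on the left hood bone, both sides, or the right, by head orientation.
# Further crouch, spine and cap-tail corrective chains are present but disconnected from execution.
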ControlRigBlueprint /BRCosmetics/Characters/Player/Male/Large/Bodies/M_LRG_TennisLeashHood/Meshes/Parts/M_LRG_TennisLeash_FaceAcc_CtrlRig
graphs [
 RigVMFunctionLibrary {links [], nodes []}
 RigVMFunction_Sequence_1_ContainedGraph {nodes [
   n1 = RigVMFunctionEntryNode()
   RigVMFunction_Sequence = &RigVMFunction_Sequence(ExecuteContext=@n1.ExecuteContext)
   n2 = &RigVMFunction_Sequence(ExecuteContext=@RigVMFunction_Sequence.B)
   n3 = &RigVMFunction_Sequence(ExecuteContext=@n2.B)
   RigVMFunctionReturnNode(A=@RigVMFunction_Sequence.A, B=@n2.A, C=@n3.A, D=@n3.B)
  ]}
 RigVMModel {nodes [
   n1 = &RigUnit_SphericalPoseReader(
    ActiveRegionScaleFactors=(PositiveWidth=0.5,NegativeWidth=1.5,PositiveHeight=0.5,NegativeHeight=0.75)
    ActiveRegionSize=0.1
    Debug=(bDrawDebug=true,bDraw2D=False,bDrawLocalAxes=False,DebugScale=25,DebugSegments=20,DebugThickness=0.25)
    DriverAxis=(X=0,Y=-1,Z=0)
    DriverItem=(Type=Bone,Name="head")
    ExecutePin=()
    FalloffRegionScaleFactors=(PositiveWidth=1,NegativeWidth=1,PositiveHeight=1,NegativeHeight=1)
    FalloffSize=0.35
    FlipHeightScaling=False
    FlipWidthScaling=False
    OptionalParentItem=(Type=Bone,Name="None")
    RotationOffset=(X=-45,Y=90,Z=0)
   )
   n2 = &RigVMFunction_AlphaInterp(InterpSpeedDecreasing=100, InterpSpeedIncreasing=100, Scale=1, bClampResult=False, bInterpResult=False, Value=@n1.OutputParam)
   &RigUnit_ModifyTransforms(
    ExecutePin=()
    ItemToModify=((Item=(Type=Bone,Name="dyn_hood_r"),Transform=(Rotation=(X=0,Y=0,Z=0,W=1),Translation=(X=-5,Y=-2.5,Z=0),Scale3D=(X=1,Y=1,Z=1))))
    ItemToModify.0.Transform.Rotation=&RigVMFunction_MathQuaternionFromEuler(Euler=(X=0,Y=0,Z=30), RotationOrder=ZXY).Result
    Weight=@n2.Result
   )
   n3 = RigVMVariableNode() {variable=is_crouch}
   Clamp = &RigVMFunction_MathDoubleClamp(Value=&RigVMFunction_MathDoubleRemap(SourceMaximum=1, TargetMaximum=1, TargetMinimum=0, bClamp=false, SourceMinimum=@n3.Value, Value=@n2.Result).Result)
   > &RigUnit_ModifyTransforms(
    ItemToModify=((Item=(Type=Bone,Name="dyn_hood_up"),Transform=(Rotation=(X=0,Y=0,Z=0,W=1),Translation=(X=7.5,Y=10,Z=0),Scale3D=(X=1,Y=1,Z=1))))
    Weight=@Clamp.Result
    ItemToModify.0.Transform.Rotation=&RigVMFunction_MathQuaternionFromEuler(Euler=(X=20,Y=0,Z=0), RotationOrder=ZXY).Result
   )
   > &RigUnit_ModifyTransforms(
    ItemToModify=((Item=(Type=Bone,Name="dyn_hood_l"),Transform=(Rotation=(X=0,Y=0,Z=0,W=1),Translation=(X=2,Y=1,Z=0),Scale3D=(X=1,Y=1,Z=1))))
    ItemToModify.0.Transform.Rotation=&RigVMFunction_MathQuaternionFromEuler(Euler=(X=0,Y=30,Z=20), RotationOrder=ZXY).Result
    Weight=@n2.Result
   )
   > &RigUnit_ModifyTransforms(
    ItemToModify=((Item=(Type=Bone,Name="dyn_cap_tail_base"),Transform=(Rotation=(X=0,Y=0,Z=0,W=1),Translation=(X=-2,Y=-2,Z=5),Scale3D=(X=1,Y=1,Z=1))))
    ItemToModify.0.Transform.Rotation=&RigVMFunction_MathQuaternionFromEuler(Euler=(X=0,Y=0,Z=-20), RotationOrder=ZXY).Result
    Weight=@n2.Result
   )
   RigUnit_BeginExecution = &RigUnit_BeginExecution()
   &RigUnit_ParentConstraint(
    AdvancedSettings=(InterpolationType=Average,RotationOrderForFilter=XZY)
    Child=(Type=Bone,Name="dyn_glasses")
    Filter=(TranslationFilter=(bX=false,bY=true,bZ=false),RotationFilter=(bX=true,bY=true,bZ=true),ScaleFilter=(bX=True,bY=True,bZ=True))
    Parents=((Item=(Type=Bone,Name="FACIAL_L_CheekOuter"),Weight=0.5),(Item=(Type=Bone,Name="FACIAL_R_CheekOuter"),Weight=0.5))
    Weight=0.75
    ExecutePin=@RigUnit_BeginExecution.ExecutePin
   )
   > RigVMFunction_Sequence = RigVMAggregateNode()
   n4 = &RigUnit_SphericalPoseReader(
    ActiveRegionScaleFactors=(PositiveWidth=0.5,NegativeWidth=0.5,PositiveHeight=1,NegativeHeight=0.5)
    ActiveRegionSize=0.1
    Debug=(bDrawDebug=true,bDraw2D=False,bDrawLocalAxes=False,DebugScale=25,DebugSegments=20,DebugThickness=0.25)
    DriverAxis=(X=0,Y=1,Z=0)
    DriverItem=(Type=Bone,Name="head")
    FalloffRegionScaleFactors=(PositiveWidth=1,NegativeWidth=1,PositiveHeight=1,NegativeHeight=1)
    FalloffSize=0.3
    FlipHeightScaling=False
    FlipWidthScaling=False
    OptionalParentItem=(Type=Bone,Name="None")
    RotationOffset=(X=150,Y=45,Z=75)
    ExecutePin=@RigVMFunction_Sequence.A
   )
   n5 = &RigVMFunction_AlphaInterp(InterpSpeedDecreasing=100, InterpSpeedIncreasing=100, Scale=1, bClampResult=False, bInterpResult=False, Value=@n4.OutputParam)
   > &RigUnit_ModifyTransforms(
    ItemToModify=((Item=(Type=Bone,Name="dyn_hood_l"),Transform=(Rotation=(X=0,Y=0,Z=0,W=1),Translation=(X=10,Y=7.5,Z=1),Scale3D=(X=1,Y=1,Z=1))))
    Weight=@n5.Result
    ItemToModify.0.Transform.Rotation=&RigVMFunction_MathQuaternionFromEuler(Euler=(X=0,Y=0,Z=-10), RotationOrder=ZXY).Result
   )
   n6 = &RigUnit_SphericalPoseReader(
    ActiveRegionScaleFactors=(PositiveWidth=0.5,NegativeWidth=0.5,PositiveHeight=0.5,NegativeHeight=0.5)
    ActiveRegionSize=0.1
    Debug=(bDrawDebug=true,bDraw2D=False,bDrawLocalAxes=False,DebugScale=25,DebugSegments=20,DebugThickness=0.25)
    DriverAxis=(X=0,Y=1,Z=0)
    DriverItem=(Type=Bone,Name="head")
    FalloffRegionScaleFactors=(PositiveWidth=1,NegativeWidth=1,PositiveHeight=1,NegativeHeight=1)
    FalloffSize=0.2
    FlipHeightScaling=False
    FlipWidthScaling=False
    OptionalParentItem=(Type=Bone,Name="None")
    RotationOffset=(X=150,Y=90,Z=90)
    ExecutePin=@RigVMFunction_Sequence.C
   )
   n7 = &RigVMFunction_AlphaInterp(InterpSpeedDecreasing=100, InterpSpeedIncreasing=100, Scale=1, bClampResult=False, bInterpResult=False, Value=@n6.OutputParam)
   > &RigUnit_ModifyTransforms(
    ItemToModify=((Item=(Type=Bone,Name="dyn_hood_l"),Transform=(Rotation=(X=0,Y=0,Z=0,W=1),Translation=(X=10,Y=5,Z=1),Scale3D=(X=1,Y=1,Z=1))))
    Weight=@n7.Result
    ItemToModify.0.Transform.Rotation=&RigVMFunction_MathQuaternionFromEuler(Euler=(X=0,Y=0,Z=-10), RotationOrder=ZXY).Result
   )
   > &RigUnit_ModifyTransforms(
    ItemToModify=((Item=(Type=Bone,Name="dyn_hood_r"),Transform=(Rotation=(X=0,Y=0,Z=0,W=1),Translation=(X=-12,Y=-6,Z=2),Scale3D=(X=1,Y=1,Z=1))))
    Weight=@n7.Result
    ItemToModify.0.Transform.Rotation=&RigVMFunction_MathQuaternionFromEuler(Euler=(X=0,Y=0,Z=-10), RotationOrder=ZXY).Result
   )
   n8 = &RigUnit_SphericalPoseReader(
    ActiveRegionScaleFactors=(PositiveWidth=0.5,NegativeWidth=0.5,PositiveHeight=1,NegativeHeight=0.5)
    ActiveRegionSize=0.1
    Debug=(bDrawDebug=true,bDraw2D=False,bDrawLocalAxes=False,DebugScale=25,DebugSegments=20,DebugThickness=0.25)
    DriverAxis=(X=0,Y=1,Z=0)
    DriverItem=(Type=Bone,Name="head")
    FalloffRegionScaleFactors=(PositiveWidth=1,NegativeWidth=1,PositiveHeight=1,NegativeHeight=1)
    FalloffSize=0.3
    FlipHeightScaling=False
    FlipWidthScaling=False
    OptionalParentItem=(Type=Bone,Name="None")
    RotationOffset=(X=160,Y=125,Z=75)
    ExecutePin=@RigVMFunction_Sequence.D
   )
   n9 = &RigVMFunction_AlphaInterp(InterpSpeedDecreasing=100, InterpSpeedIncreasing=100, Scale=1, bClampResult=False, bInterpResult=False, Value=@n8.OutputParam)
   > &RigUnit_ModifyTransforms(
    ItemToModify=((Item=(Type=Bone,Name="dyn_hood_r"),Transform=(Rotation=(X=0,Y=0,Z=0,W=1),Translation=(X=-10,Y=-5,Z=0),Scale3D=(X=1,Y=1,Z=1))))
    Weight=@n9.Result
    ItemToModify.0.Transform.Rotation=&RigVMFunction_MathQuaternionFromEuler(Euler=(X=0,Y=0,Z=-10), RotationOrder=ZXY).Result
   )
   n10 = &RigUnit_SphericalPoseReader(
    ActiveRegionScaleFactors=(PositiveWidth=0.1,NegativeWidth=0.1,PositiveHeight=0.5,NegativeHeight=0.1)
    ActiveRegionSize=0.1
    Debug=(bDrawDebug=true,bDraw2D=False,bDrawLocalAxes=False,DebugScale=25,DebugSegments=20,DebugThickness=0.25)
    DriverAxis=(X=0,Y=0,Z=1)
    DriverItem=(Type=Bone,Name="spine_05")
    ExecutePin=()
    FalloffRegionScaleFactors=(PositiveWidth=1,NegativeWidth=1,PositiveHeight=1,NegativeHeight=1)
    FalloffSize=0.125
    FlipHeightScaling=False
    FlipWidthScaling=False
    OptionalParentItem=(Type=Bone,Name="None")
    RotationOffset=(X=-20,Y=0,Z=0)
   )
   Multiply = &RigVMFunction_MathDoubleMul(
    B=&RigVMFunction_MathDoubleRemap(SourceMaximum=1, SourceMinimum=0, TargetMaximum=0.5, TargetMinimum=0, Value=@n10.OutputParam).Result
    A=@n2.Result
   )
   > &RigUnit_ModifyTransforms(
    ItemToModify=((Item=(Type=Bone,Name="dyn_hood_r"),Transform=(Rotation=(X=0,Y=0,Z=0,W=1),Translation=(X=0,Y=5,Z=-2),Scale3D=(X=1,Y=1,Z=1))))
    ItemToModify.0.Transform.Rotation=&RigVMFunction_MathQuaternionFromEuler(Euler=(X=0,Y=0,Z=20), RotationOrder=ZXY).Result
    Weight=@Multiply.Result
   )
   > &RigUnit_ModifyTransforms(
    ItemToModify=((Item=(Type=Bone,Name="dyn_hood_up"),Transform=(Rotation=(X=0,Y=0,Z=0,W=1),Translation=(X=-5,Y=-5,Z=0),Scale3D=(X=1,Y=1,Z=1))))
    ItemToModify.0.Transform.Rotation=&RigVMFunction_MathQuaternionFromEuler(Euler=(X=0,Y=0,Z=60), RotationOrder=ZXY).Result
    Weight=@Multiply.Result
   )
   Divide = RigVMUnitNode(A=(), B=()) {template="Divide::Execute(in A,in B,out Result)"}
   &RigUnit_ModifyTransforms(
    ExecutePin=()
    ItemToModify=((Item=(Type=Bone,Name="dyn_hood_l"),Transform=(Rotation=(X=0,Y=0,Z=0,W=1),Translation=(X=0,Y=-30,Z=0),Scale3D=(X=1,Y=1,Z=1))))
    ItemToModify.0.Transform.Rotation=&RigVMFunction_MathQuaternionFromEuler(Euler=(X=0,Y=0,Z=0), RotationOrder=ZXY).Result
    Weight=@Multiply.Result
   )
   &RigVMFunction_MathDoubleMul(B=RigVMVariableNode() {variable=invert_crouch}.Value, A=@Multiply.Result)
   &RigVMFunction_MathDoubleMul(B=@Multiply.Result, A=@n3.Value)
   Subtract = &RigVMFunction_MathDoubleSub(A=1, B=@n10.OutputParam)
   RigVMVariableNode() {variable=is_crouch}
  ]}
]
hierarchy [
 root {loc=<0 0 0>, rot=<0 0 0 1>}
 pelvis {parent=root, loc=<0 -0.7321 92.4218>, rot=<0.0127 0.707 -0.0127 -0.707>}
 spine_01 {parent=pelvis, loc=<5.6002 -0.5128 0>, rot=<0 0 0.034 -0.9994>}
 spine_02 {parent=spine_01, loc=<8.0756 0.0139 0>, rot=<0 0 0.0238 -0.9997>}
 spine_03 {parent=spine_02, loc=<7.0395 0 0>, rot=<0 0 0.0149 -0.9999>}
 spine_04 {parent=spine_03, loc=<8.1024 0 0>, rot=<0 0 -0.0132 -0.9999>}
 spine_05 {parent=spine_04, loc=<10.0141 0 0>, rot=<0 0 -0.0015 -1>}
 clavicle_l {parent=spine_05, loc=<2.5856 -0.2866 -4.3777>, rot=<0.0465 -0.7289 -0.009 -0.6829>}
 upperarm_l {parent=clavicle_l, loc=<18.2264 0.1342 -0.001>, rot=<0.0253 -0.2693 0.0013 -0.9627>}
 neck_01 {parent=spine_05, loc=<10.815 1.9196 0>, rot=<0 0 -0.1864 -0.9825>}
 neck_02 {parent=neck_01, loc=<4.7347 0.3045 0>, rot=<0 0 0.0363 -0.9993>}
 head {parent=neck_02, loc=<5.7632 0.5552 0>, rot=<0 0 0.115 -0.9934>}
 dyn_hood_up {parent=head, loc=<15.0821 4.9443 0>, rot=<0.7062 -0.708 0 0>}
 FACIAL_C_FacialRoot {parent=head, loc=<2.1743 -3.3889 -0.0002>, rot=<0.4879 -0.5118 -0.4879 -0.5118>}
 FACIAL_L_CheekInner {parent=FACIAL_C_FacialRoot, loc=<1.9467 -1.1328 9.1888>, rot=<-0.121 -0.0883 -0.1969 -0.9689>}
 FACIAL_R_CheekInner {parent=FACIAL_C_FacialRoot, loc=<-1.9371 -1.1306 9.1864>, rot=<-0.121 0.0886 0.1969 -0.9689>}
 FACIAL_L_CheekOuter {parent=FACIAL_C_FacialRoot, loc=<3.7448 -0.8667 7.7388>, rot=<0.0573 -0.3766 0.1012 -0.919>}
 FACIAL_R_CheekOuter {parent=FACIAL_C_FacialRoot, loc=<-3.7422 -0.8649 7.7359>, rot=<0.0573 0.3768 -0.1012 -0.919>}
 FACIAL_L_NasolabialBulge {parent=FACIAL_C_FacialRoot, loc=<3.6059 2.795 7.5698>, rot=<0.06 -0.3772 0.0247 -0.9239>}
 FACIAL_R_NasolabialBulge {parent=FACIAL_C_FacialRoot, loc=<-3.593 2.7823 7.5436>, rot=<0.06 0.3774 -0.0246 -0.9238>}
 FACIAL_L_NasolabialFurrow {parent=FACIAL_C_FacialRoot, loc=<2.8511 1.203 10.1415>, rot=<-0.0311 -0.1265 -0.3633 -0.9225>}
 FACIAL_R_NasolabialFurrow {parent=FACIAL_C_FacialRoot, loc=<-2.8329 1.2067 10.1365>, rot=<-0.0312 0.1267 0.3634 -0.9225>}
 FACIAL_L_CheekLower {parent=FACIAL_C_FacialRoot, loc=<3.9462 2.5381 6.783>, rot=<0.0366 -0.4266 0.0587 -0.9018>}
 FACIAL_R_CheekLower {parent=FACIAL_C_FacialRoot, loc=<-3.9216 2.5237 6.7531>, rot=<0.0367 0.4268 -0.0587 -0.9017>}
 FACIAL_C_Nose {parent=FACIAL_C_FacialRoot, loc=<0.0016 -1.6257 10.831>, rot=<-0.1281 0.0001 0 -0.9918>}
 FACIAL_C_NoseLower {parent=FACIAL_C_Nose, loc=<-0.0015 2.5372 -0.0469>, rot=<0.1305 0 0 -0.9914>}
 FACIAL_L_Nostril {parent=FACIAL_C_Nose, loc=<1.4777 1.949 -0.2497>, rot=<0.0173 0.1305 0.0023 -0.9913>}
 FACIAL_R_Nostril {parent=FACIAL_C_Nose, loc=<-1.4972 1.9631 -0.2657>, rot=<0.0173 -0.1305 -0.0023 -0.9913>}
 FACIAL_C_NoseTip {parent=FACIAL_C_Nose, loc=<0.01 1.7123 0.8358>, rot=<0.0872 0 0 -0.9962>}
 FACIAL_C_MouthUpper {parent=FACIAL_C_FacialRoot, loc=<0.0012 3.0528 8.6538>, rot=<0.0023 0.0001 0 -1>}
 FACIAL_C_LipUpper {parent=FACIAL_C_MouthUpper, loc=<-0.0064 -0.8354 2.7306>, rot=<-0.0872 0 0 -0.9962>}
 FACIAL_L_LipUpper {parent=FACIAL_C_MouthUpper, loc=<1.4885 -0.6048 2.3748>, rot=<-0.0778 -0.1601 -0.0379 -0.9833>}
 FACIAL_R_LipUpper {parent=FACIAL_C_MouthUpper, loc=<-1.5389 -0.6155 2.3747>, rot=<-0.0778 0.1601 0.0379 -0.9833>}
 FACIAL_L_LipUpperOuter {parent=FACIAL_C_MouthUpper, loc=<2.2363 -0.1996 1.6568>, rot=<-0.0984 -0.2293 -0.0983 -0.9634>}
 FACIAL_R_LipUpperOuter {parent=FACIAL_C_MouthUpper, loc=<-2.3072 -0.2224 1.6291>, rot=<-0.0984 0.2293 0.0983 -0.9634>}
 FACIAL_L_LipCorner {parent=FACIAL_C_MouthUpper, loc=<2.4795 0.2213 0.8899>, rot=<0 -0.2078 0 -0.9782>}
 FACIAL_R_LipCorner {parent=FACIAL_C_MouthUpper, loc=<-2.5414 0.1972 0.8511>, rot=<0 0.2078 0 -0.9782>}
 FACIAL_L_Masseter {parent=FACIAL_C_FacialRoot, loc=<6.3133 -0.2766 5.0208>, rot=<0.0716 0.113 0.0076 -0.991>}
 FACIAL_R_Masseter {parent=FACIAL_C_FacialRoot, loc=<-6.3072 -0.2827 5.0054>, rot=<0.0716 -0.1128 -0.0076 -0.991>}
 FACIAL_C_UnderChin {parent=FACIAL_C_FacialRoot, loc=<0.0102 8.2736 5.7884>, rot=<0.5018 0.0001 0.0001 -0.865>}
 FACIAL_L_UnderChin {parent=FACIAL_C_FacialRoot, loc=<3.7526 6.6589 3.9389>, rot=<0.4044 -0.3546 0.0832 -0.8389>}
 FACIAL_R_UnderChin {parent=FACIAL_C_FacialRoot, loc=<-3.7309 6.6415 3.9307>, rot=<0.4045 0.3548 -0.0831 -0.8388>}
 FACIAL_C_LowerLipRotation {parent=FACIAL_C_FacialRoot, loc=<0.0273 0.4183 3.4679>, rot=<0.3442 0.0001 0.0001 -0.9389>}
 FACIAL_C_MouthLower {parent=FACIAL_C_LowerLipRotation, loc=<-0.025 -1.3415 5.66>, rot=<-0.342 0 0 -0.9397>}
 FACIAL_C_LipLower {parent=FACIAL_C_MouthLower, loc=<0.0055 1.133 2.4349>, rot=<0.0872 0 0 -0.9962>}
 FACIAL_L_LipLower {parent=FACIAL_C_MouthLower, loc=<1.3642 1.1041 2.2072>, rot=<0.0927 -0.1294 0.0272 -0.9869>}
 FACIAL_R_LipLower {parent=FACIAL_C_MouthLower, loc=<-1.3745 1.056 2.1891>, rot=<0.0927 0.1294 -0.0272 -0.9869>}
 FACIAL_L_LipLowerOuter {parent=FACIAL_C_MouthLower, loc=<2.2291 0.6329 1.4725>, rot=<0.0913 -0.2321 0.0318 -0.9679>}
 FACIAL_R_LipLowerOuter {parent=FACIAL_C_MouthLower, loc=<-2.2563 0.6228 1.4435>, rot=<0.0913 0.2321 -0.0318 -0.9679>}
 FACIAL_C_Jaw {parent=FACIAL_C_FacialRoot, loc=<0.0273 0.4183 3.4679>, rot=<0.3442 0.0001 0.0001 -0.9389>}
 FACIAL_C_Jawline {parent=FACIAL_C_Jaw, loc=<-0.0218 3.1798 8.2939>, rot=<0.2586 0 0 -0.966>}
 FACIAL_L_ChinSide {parent=FACIAL_C_Jaw, loc=<3.4231 0.8203 7.7563>, rot=<-0.2208 -0.3738 0.0499 -0.8995>}
 FACIAL_R_ChinSide {parent=FACIAL_C_Jaw, loc=<-3.4811 0.83 7.7482>, rot=<-0.2208 0.3738 -0.0499 -0.8995>}
 FACIAL_C_Chin {parent=FACIAL_C_Jaw, loc=<-0.0233 1.3228 8.3122>, rot=<-0.2591 0 0 -0.9659>}
 dyn_beard_base {parent=FACIAL_C_Chin, loc=<-0.0032 1.853 1.2848>, rot=<0.6422 0.2967 0.6397 0.3005>}
 dyn_beard {parent=dyn_beard_base, loc=<0 0 0>, rot=<0 0 0 1>}
 dyn_beard_l {parent=dyn_beard, loc=<0 0 6.5528>, rot=<0 0 0 1>}
 dyn_beard_r {parent=dyn_beard, loc=<-0.0719 0.0467 -6.5522>, rot=<0.0055 -0.0036 1 0>}
 FACIAL_L_Ear {parent=FACIAL_C_FacialRoot, loc=<6.5389 -1.7666 2.6306>, rot=<-0.024 -0.0006 -0.0261 -0.9994>}
 FACIAL_R_Ear {parent=FACIAL_C_FacialRoot, loc=<-6.5123 -1.7725 2.6576>, rot=<-0.024 0.0009 0.0262 -0.9994>}
 FACIAL_L_Jawline {parent=FACIAL_C_FacialRoot, loc=<4.8558 4.4269 5.2379>, rot=<0.2115 0.1577 -0.055 -0.963>}
 FACIAL_R_Jawline {parent=FACIAL_C_FacialRoot, loc=<-4.8264 4.4172 5.2234>, rot=<0.2115 -0.1575 0.0551 -0.963>}
 dyn_glasses {parent=head, loc=<0 0 -0.0002>, rot=<0.0029 -0.7071 -0.0029 -0.7071>}
 dyn_hood_l {parent=neck_02, loc=<5.1443 5.2001 -10.5593>, rot=<0.5902 -0.8073 0 0>}
 dyn_hood_r {parent=neck_02, loc=<5.1236 5.1348 10.5593>, rot=<0.8073 0.5902 0 0>}
 clavicle_r {parent=spine_05, loc=<2.5855 -0.2866 4.3777>, rot=<0.7289 0.0465 -0.6829 0.009>}
 upperarm_r {parent=clavicle_r, loc=<-18.2264 -0.1342 0.001>, rot=<0.0253 -0.2693 0.0013 -0.9627>}
]
variables [
 invert_crouch {type=double, default=0}
 is_crouch {type=double, default=0}
]
