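# Post-copy AnimBlueprint for the face accessory part of the large male TennisLeashHood character body, adding procedural secondary motion.
# Copies the attached parent's pose, runs a Control Rig, then a RigidBody simulation in spine_05 space using an override physics asset.
# Before simulation, an additive dyn_beard_base offset applies only in the front end, its alpha set each update from bIsInFrontEnd.
# Hood AnimDynamics and a duplicate beard transform are left unconnected.
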AnimBlueprint /BRCosmetics/Characters/Player/Male/Large/Bodies/M_LRG_TennisLeashHood/Meshes/Parts/M_LRG_TennisLeashHood_FaceAcc_AnimBP
target_skeleton=&/M_LRG_TennisLeashHood_FaceAcc_Skeleton
graphs [
 AnimGraph {nodes [
   Anim Dynamics - Bone: dyn_hood_up(Alpha=0.5, ExternalForce="0,0,0") {
    angularDampingOverride=1
    angularSpringConstant=500
    bAngularSpring=true
    bGravityOverrideInSimSpace=true
    bOverrideAngularDamping=true
    simulationSpace=BoneRelative
    alphaBoolBlend {blendOption=Linear}
    alphaScaleBiasClamp {bInterpResult=true, interpSpeedDecreasing=50, interpSpeedIncreasing=50}
    boundBone {boneName=dyn_hood_up}
    componentAppliedLinearAccClamp=<500 500 500>
    componentLinearAccScale=<0.5 0.5 0.5>
    componentLinearVelScale=<1 1 1>
    gravityOverride {z=-980}
    physicsBodyDefinitions [{
      boundBone {boneName=dyn_hood_up}
      boxExtents=<30 40 50>
      constraintSetup {angularLimitsMax=<0 0 0>, angularLimitsMin=<0 0 0>, angularTarget=<0 1 0>, linearAxesMax=<5 0 0>, linearAxesMin=<-5 0 0>}
      localJointOffset=<1 0 0>
     }]
    relativeSpaceBone {boneName=spine_04}
   }
   "Transform (Modify) Bone - Bone: dyn_beard_base"(Alpha=1, Translation="-1,2,0", Rotation="0.000,-30.000,-0.000", Scale="1,1,1") {
    rotationMode=BMM_Additive
    rotationSpace=BCS_BoneSpace
    translationMode=BMM_Additive
    translationSpace=BCS_BoneSpace
    alphaBoolBlend {blendOption=Linear}
    boneToModify {boneName=dyn_beard_base}
    rotation {yaw=-29.9999}
    translation=<-1 2>
   }
   n1 = Control Rig(Source=Copy Pose From Mesh() {bUseAttachedParent=true}) {
    alphaBoolBlend {blendOption=Linear}
    alphaScaleBiasClamp {bInterpResult=true}
    controlRigAssetReference {blueprintRigClass=&/M_LRG_TennisLeash_FaceAcc_CtrlRig.M_LRG_TennisLeash_FaceAcc_CtrlRig_C}
   }
   n2 = "Transform (Modify) Bone - Bone: dyn_beard_base"(
    ComponentPose=Local To Component(LocalPose=@n1)
    Alpha=Get Dyn_FrontEnd_Alpha() -> (Dyn_FrontEnd_Alpha:float)
    Translation="-1,2,0"
    Rotation="0.000,-30.000,-0.000"
    Scale="1,1,1"
   ) {
    rotationMode=BMM_Additive
    rotationSpace=BCS_BoneSpace
    translationMode=BMM_Additive
    translationSpace=BCS_BoneSpace
    alphaBoolBlend {blendOption=Linear}
    boneToModify {boneName=dyn_beard_base}
    rotation=<rot 0 -29.9999 0>
    translation=<-1 2>
   }
   n3 = RigidBody(ComponentPose=@n2) {
    bOverrideWorldGravity=true
    overridePhysicsAsset=&/M_LRG_TennisLeashHood_FaceAcc_Physics
    simulationSpace=BaseBoneSpace
    alphaBoolBlend {blendOption=Linear}
    alphaScaleBiasClamp {bInterpResult=true}
    baseBoneRef {boneName=spine_05}
    componentAppliedLinearAccClamp=<1250 1250 1250>
    componentLinearAccScale=<0.75 0.75 0.75>
    componentLinearVelScale=<0.5 0.5 0.5>
    overrideWorldGravity {z=-980}
    simSpaceSettings {maxAngularAcceleration=10, maxAngularVelocity=5, maxLinearAcceleration=1, maxLinearVelocity=1, velocityScaleZ=0.25, worldAlpha=1}
   }
   Output Pose(Result=Component To Local(ComponentPose=@n3)) {layerGroup=None}
  ]}
 EventGraph {nodes [
   Event BlueprintUpdateAnimation()
   > n1 = Branch(Condition=Get bIsInFrontEnd())
   > Set Dyn_FrontEnd_Alpha(Dyn_FrontEnd_Alpha:float=1) -> (Output_Get:float)
   Set Dyn_FrontEnd_Alpha(execute=@n1.else, Dyn_FrontEnd_Alpha=0)
  ]}
]
variables [Dyn_FrontEnd_Alpha {type=number, default=1}]
